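# Idle energy effect for the VerTet backpack cosmetic, with variant-selected colors, time-of-day brightness and separate front-end timing or intensity.
# Energy builds four ribbons from sockets 0-3 toward socket 4 using chained Hermite splines, camera-vector sine wobble and platform-dependent resolution.
# Center_Energy repeatedly spawns fading meshes at socket 4.
# Glow_Packed packs socket-attached lights, sprites and meshes into one emitter via spawn-group visibility tags, gating some groups by platform.
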NiagaraSystem /BRCosmetics/Accessories/FORT_Backpacks/Backpack_M_MED_VerTet/FX/NS_Backpack_VerTet_EnergyIdle
dependencies {
 usedDataInterfaces [
  &NiagaraDataInterfacePlatformSet
  &NiagaraDataInterfaceCurve
  &NiagaraDataInterfaceCamera
  &NiagaraDataInterfaceSocketReader
 ]
 usedDynamicInputs [&FNTimeOfDayScalar]
 usedModules [
  &self:SystemSpawnScript.SystemScriptSource.SystemScriptGraph.NiagaraNodeAssignment_3.SetVariables_#1
  &SystemState
  &FN_VariantManager
  &self:SystemSpawnScript.SystemScriptSource.SystemScriptGraph.NiagaraNodeAssignment_0.SetVariables_#2
  &self:SystemSpawnScript.SystemScriptSource.SystemScriptGraph.NiagaraNodeAssignment_1.SetVariables_#3
 ]
}
emitters {
 Center_Energy {
  emitterUpdateScript [
   &EmitterState(Inactive Response=::NewEnumerator1, Life Cycle Mode=::NewEnumerator1, Loop Duration=0.3, UseLoopDelay=false)
   &SpawnBurst_Instantaneous(Use Spawn Probability=false, UseLoopCountLimit=false)
   &SpawnRate(SpawnRate=2, Use Spawn Probability=false)
  ]
  particleSpawnScript [
   &InitializeParticle(
    Color Mode=::NewEnumerator0
    Lifetime=&RandomRangeFloat(Minimum=0.75, Maximum=1)
    Mesh Scale Mode=::NewEnumerator3
    Mesh Uniform Scale=0.065
    UsePositionOffset=false
    Write Material Random=true
   )
   &InitialMeshOrientation(
    Mesh Orientation Mode=::NewEnumerator7
    Rotation=&MakeVector(X=0, Y=0, Z=&RandomRangeFloat(Minimum=0, Maximum=1))
    Use Rotation Vector=true
   )
   &CameraOffset(Camera Offset Amount=0.5)
  ]
  particleUpdateScript [
   &ParticleState()
   &SocketLocation(
    Orientation Sampling=::NewEnumerator1
    SamplePosOffset=false
    Socket Index=4
    Socket Reader=$System.Socket Reader
    Socket Sampling Mode=::NewEnumerator3
    VelocitySpeedLimit=false
   )
   &DynamicMaterialParameters(
    Index 0 Param 1=&FloatFromCurve(FloatCurve=DataInterface() {propertyValues {Curve {keys [<at 0 -0.176 RCIM_Cubic>,<at 1 0.356 RCIM_Cubic>]}}})
    Index 0 Param 2=&FloatFromCurve(FloatCurve=DataInterface() {propertyValues {Curve {keys [{arriveTangent=-0.3016, interpMode=RCIM_Cubic, time=0, value=0.6}, <at 1 1.5 RCIM_Cubic>]}}})
    Index 0 Param 3=&FloatFromCurve(FloatCurve=DataInterface() {propertyValues {Curve {keys [{arriveTangent=2.2165, interpMode=RCIM_Cubic, time=0, value=0.1}, <at 1 0.35 RCIM_Cubic>]}}})
    Index 0 Param 4=1
    Module.Index0Float0Write=true
    Module.Index0Float1Write=true
    Module.Index0Float2Write=true
    Module.Index0Float3Write=false
   )
   &ScaleColor(
    Scale Alpha=&FloatFromCurve(
     FloatCurve=DataInterface() {propertyValues {Curve {keys [
         <at 0 0 RCIM_Cubic>
         {arriveTangentWeight=0.0943, interpMode=RCIM_Cubic, leaveTangentWeight=0.3771, tangentMode=RCTM_User, time=0.35, value=1}
         <at 1 0 RCIM_Cubic>
        ]}}}
     Scale Curve=2
    )
    Scale RGB=&Multiply_VectorByFloat(
     Vector=&MakeVectorFromLinearColorRGB(Color=&VariantColorSelector(
       Variant Amount=4
       Original Color=<rgba 0.2667 0 1 1>
       Variant 1 Color=<rgba 0.2957 0.2174 1 1>
       Variant 2 Color=<rgba 0.1304 0.0565 0 1>
       Variant 3 Color=<rgba 0.1667 0 1 1>
       Variant 4 Color=<rgba 0 0.5391 0.3145 1>
       Variant 5 Color=<rgba 0.4167 0 1 1>
      ))
     Float=&FloatFromCurve(
      FloatCurve=DataInterface() {propertyValues {Curve {keys [<at 0 0 RCIM_Cubic>,<at 0.2 1 RCIM_Cubic>,<at 1 0 RCIM_Cubic>]}}}
      Scale Curve=&Multiply_Float(
       A=5
       B=&Multiply_Float(
        A=$System.TODScalar
        B=&FN_FrontEndFloatModifier(
         Gameplay=&FloatFromCurve(FloatCurve=DataInterface() {propertyValues {Curve {keys [<at 0.45 0>,<at 1.45 1>]}}}, CurveIndex=$Engine.System.Age)
         Front End=&FloatFromCurve(FloatCurve=DataInterface() {propertyValues {Curve {keys [<at 0.8 0>,<at 1.75 1>]}}}, CurveIndex=$Engine.System.Age)
        )
       )
      )
     )
    )
    ScaleA=true
    ScaleRGB=true
   )
  ]
  rendererClasses [&NiagaraMeshRendererProperties]
  renderers [{rendererClass=&NiagaraMeshRendererProperties, rendererIndex=0}]
 }
 Energy {
  emitterSpawnScript [/SetVariables(Emitter.RibbonResolution:NiagaraInt32=&MakeIntFromFloat(Float=&FN_PickFloatDependOnPlatform(TargetPlatformActive=8, TargetPlatformNoActive=4))) {isSetParameters=true}]
  emitterUpdateScript [
   &EmitterState(
    Inactive Response=::NewEnumerator1
    Life Cycle Mode=::NewEnumerator1
    Loop Behavior=::NewEnumerator1
    Loop Duration Mode=::NewEnumerator1
    UseLoopDelay=false
   )
   &SpawnBurst_Instantaneous(
    Spawn Count=&Multiply_Int(A=$Emitter.RibbonResolution, B=4)
    Use Spawn Probability=false
    UseLoopCountLimit=false
   )
   /SetVariables(
    Emitter.BoltColor:LinearColor=&VariantColorSelector(
     Variant Amount=4
     Original Color=<rgba 0.4671 0.05 0.5 1>
     Variant 1 Color=<rgba 0.15 0.3075 0.5 1>
     Variant 2 Color=<rgba 0.6 0.2866 0.047 1>
     Variant 3 Color=<rgba 0.5 0.0565 0.5 1>
     Variant 4 Color=<rgba 0.0635 0.487 0.3105 1>
     Variant 5 Color=<rgba 0.288 0 0.864 1>
    )
    Emitter.GlowColor:LinearColor=&VariantColorSelector(
     Variant Amount=4
     Original Color=<rgba 0.1733 0 0.8 1>
     Variant 1 Color=<rgba 0.3216 0.5046 0.8 1>
     Variant 2 Color=<rgba 0.8 0.2438 0.0416 1>
     Variant 3 Color=<rgba 0.4686 0.1591 0.5304 1>
     Variant 4 Color=<rgba 0.093 0.713 0.5064 1>
     Variant 5 Color=<rgba 0.208 0.04 0.4 1>
    )
    Emitter.InnerGlowColor:LinearColor=&VariantColorSelector(
     Variant Amount=4
     Original Color=<rgba 12.7898 0.39 13 5>
     Variant 1 Color=<rgba 1.6522 3.3819 5 1>
     Variant 2 Color=<rgba 1.5 1.187 0.2478 1>
     Variant 3 Color=<rgba 5 1.1304 5 1>
     Variant 4 Color=<rgba 0.2616 1.88 1.5024 1>
     Variant 5 Color=<rgba 5 0.6957 5 1>
    )
    Emitter.SecondaryGlowColor:LinearColor=&VariantColorSelector(
     Variant Amount=4
     Original Color=<rgba 0.1387 0.008 0.4 1>
     Variant 1 Color=<rgba 0.18 0.257 0.4 1>
     Variant 2 Color=<rgba 0.6783 0.2917 0.0678 1>
     Variant 3 Color=<rgba 0.5304 0.1591 0.5304 1>
     Variant 4 Color=<rgba 0.0522 0.4 0.2551 1>
     Variant 5 Color=<rgba 0.2302 0.0487 0.4 1>
    )
   ) {isSetParameters=true}
  ]
  particleSpawnScript [
   &PartitionParticles(
    Normalized Index In Partition Includes One=true
    Normalized Index Includes One=true
    Number Of Partitions=4
    Particles Per Partition:NiagaraInt32=$Emitter.RibbonResolution
    Partition Mode:ENiagara_PartitionMode_v2=::NewEnumerator1
    Write Partition Index to Ribbon Link Order:ENiagara_AttributeSamplingApplyOutput=::NewEnumerator0
    Write Partition to Ribbon ID=::NewEnumerator0
   )
   &InitializeParticle(
    Color=<rgba 0.4 0.4 0.4 1>
    Lifetime=3
    Ribbon Width=2
    Ribbon Width Mode=::NewEnumerator1
    UsePositionOffset=false
    Write Material Random=true
   )
  ]
  particleUpdateScript [
   &ParticleState(Kill Particles When Lifetime Has Elapsed=false)
   /SetVariables(Particles.ProgressAlpha:NiagaraFloat=&Multiply_Float(
     A=&ModuloFloat(
      A=&Add_Float(
       A=&Multiply_Float(A=$Particles.Age, B=1)
       B=&Multiply_FloatByInt(
        Float=&Divide_Float(A=1, B=&MakeFloatFromInt(Int=4))
        Integer=$Particles.PartitionParticles.Partition
       )
      )
      B=1
     )
     B=&FN_FrontEndFloatModifier(
      Gameplay=&FloatFromCurve(FloatCurve=DataInterface() {propertyValues {Curve {keys [<at 0.35 0>,<at 1.35 1>]}}}, CurveIndex=$Engine.System.Age)
      Front End=&FloatFromCurve(FloatCurve=DataInterface() {propertyValues {Curve {keys [<at 0.7 0>,<at 1.65 1>]}}}, CurveIndex=$Engine.System.Age)
     )
    )) {isSetParameters=true}
   /SetVariables(Particles.RandomPerPartition:NiagaraFloat=&RandomRangeFloat(
     Minimum=1
     Maximum=2
     Random Seed=&Multiply_Int(A=$Particles.PartitionParticles.Partition, B=&MakeIntFromFloat(Float=$Emitter.LoopedAge))
     Fixed Random Seed=true
    )) {isSetParameters=true}
   &SocketLocation(
    Orientation Sampling=::NewEnumerator1
    Position Sampling=::NewEnumerator1
    SamplePosOffset=true
    Sampled Position Offset=<0 0 -2>
    Socket Index=&ModuloInt(A=$Particles.PartitionParticles.Partition, B=4)
    Socket Reader=$System.Socket Reader
    Socket Sampling Mode=::NewEnumerator3
    VelocitySpeedLimit=false
   )
   &SocketLocation(
    Orientation Sampling=::NewEnumerator1
    Position Sampling=::NewEnumerator1
    SamplePosOffset=true
    Sampled Position Offset=&MergeHelper_VectorSwitch(
     0=<1 1 0>
     1=<-1 1 0>
     2=<-1 -1 0>
     3=<1 -1 0>
     Index=&ModuloInt(A=$Particles.PartitionParticles.Partition, B=4)
     Num=::NewEnumerator2
    )
    Socket Index=4
    Socket Reader=$System.Socket Reader
    Socket Sampling Mode=::NewEnumerator3
    VelocitySpeedLimit=false
   ) {module=SocketLocation001}
   /Game/Effects/Niagara/Modules/Curves/FN_HermiteSpline(
    Direction Start:Vector3f=&Multiply_VectorByFloat(
     Vector=&MergeHelper_VectorSwitch(
      0=<0 0 6.0209>
      1=<0 0 5.8634>
      2=<0 0 6.3454>
      3=<0 0 6.2515>
      Index=&ModuloInt(A=$Particles.PartitionParticles.Partition, B=4)
      Num=::NewEnumerator2
     )
     Float=1
    )
    Position Blend:NiagaraFloat=$Particles.PartitionParticles.Normalized Index In Partition
    Position End:Vector3f=&Add_Vector(
     A=Linked() {linkedVariable {name=Output.SocketLocation.SampledPosition, type=NiagaraPosition}}
     B=&RotateVectorByQuaternion(Vector To Rotate=<0 0 12>, Quaternion=$Output.SocketLocation.SampledRotation)
    )
    Position Start:NiagaraPosition=$Output.SocketLocation.SampledPosition
    Write Particles Position=true
   )
   &FN_HermiteSpline(
    Direction End:Vector3f=&MakeVector(
     X=0
     Y=0
     Z=&FloatFromCurve(
      FloatCurve=DataInterface() {propertyValues {Curve {keys [<at 0 0>,<at 1 1>]}}}
      CurveIndex=$Particles.ProgressAlpha
      Scale Curve=&Multiply_Float(A=45, B=$Particles.RandomPerPartition)
     )
    )
    Direction Start:Vector3f=&MakeVector(
     X=0
     Y=0
     Z=&FloatFromCurve(
      FloatCurve=DataInterface() {propertyValues {Curve {keys [<at 0 0>,<at 1 1>]}}}
      CurveIndex=$Particles.ProgressAlpha
      Scale Curve=&Multiply_Float(A=-9, B=$Particles.RandomPerPartition)
     )
    )
    Position Blend:NiagaraFloat=$Particles.PartitionParticles.Normalized Index In Partition
    Position End:NiagaraPosition=$Output.SocketLocation001.SampledPosition
    Position Start:Vector3f=$Output.FN_HermiteSpline.Position
    Write Particles Position=true
   ) {module=FN_HermiteSpline001}
   /SetVariables(Particles.Position=&AddVectorToPosition(
     Position=$Particles.Position
     Vector=&Multiply_VectorByFloat(
      Vector=&Lerp_Vector(
       A=&Multiply_VectorByFloat(Vector=&ReturnCameraVector(Camera Vectors=::NewEnumerator2, Coordinate Space=::Local), Float=1.5)
       B=&Multiply_VectorByFloat(Vector=&ReturnCameraVector(Camera Vectors=::NewEnumerator3, Coordinate Space=::Local), Float=0.5)
       Alpha=$Particles.PartitionParticles.Normalized Index In Partition
      )
      Float=&Sine(
       Normalized Angle=&Add_Float(
        A=&Multiply_Float(A=$Particles.PartitionParticles.Normalized Index In Partition, B=&Multiply_Float(A=30, B=3.14))
        B=&Add_Float(
         A=&RandomRangeFloat(Minimum=0, Maximum=20, Randomness Mode=::NewEnumerator1, Random Seed=$Particles.PartitionParticles.Partition, Fixed Random Seed=true)
         B=&Multiply_Float(A=$Engine.Time, B=1.5)
        )
       )
       Period=1
       Scale=&FloatFromCurve(
        FloatCurve=DataInterface() {propertyValues {Curve {keys [
            {arriveTangent=3.7981, interpMode=RCIM_Cubic, leaveTangent=3.7981, tangentMode=RCTM_User, time=0.15, value=0}
            {arriveTangent=0.527, interpMode=RCIM_Cubic, leaveTangent=0.527, tangentMode=RCTM_User, time=0.3, value=0.85}
            {
             arriveTangentWeight=0.2045
             interpMode=RCIM_Cubic
             leaveTangent=-13.628
             leaveTangentWeight=0.0341
             tangentMode=RCTM_Break
             time=0.8
             value=1
            }
           ]}}}
        CurveIndex=$Particles.PartitionParticles.Normalized Index In Partition
       )
       Bias=0
      )
     )
    )) {isSetParameters=true}
   &ScaleRibbonWidth(Ribbon Width Scale=&FloatFromCurve(
     FloatCurve=DataInterface() {propertyValues {Curve {keys [<at 0 0>,<at 0.24 0.38>,<at 0.3 0.95>,<at 1 1.15>]}}}
     CurveIndex=$Particles.PartitionParticles.Normalized Index In Partition
    ))
   &ScaleColor(
    Scale Mode=1
    Scale RGBA=&MakeVector4_FromLinearColor(LinearColor=&Multiply_LinearColorByFloat(Linear Color=<rgba 1 1 1 2>, Float=$System.TODScalar))
    ScaleRGBA=true
   )
   &DynamicMaterialParameters(
    Index 1 Param 2=2
    Index 1 Param 3=&FloatFromCurve(
     FloatCurve=DataInterface() {propertyValues {Curve {keys [
         <at 0 0.668 RCIM_Cubic>
         {arriveTangentWeight=0.1707, interpMode=RCIM_Cubic, leaveTangentWeight=0.2275, tangentMode=RCTM_User, time=0.3, value=-0.288}
         {arriveTangentWeight=0.1693, interpMode=RCIM_Cubic, leaveTangentWeight=0.2257, tangentMode=RCTM_User, time=0.7, value=-0.288}
         <at 1 0.668 RCIM_Cubic>
        ]}}}
     CurveIndex=$Particles.ProgressAlpha
    )
    Module.Index0Float0Write=true
    Module.Index0Float1Write=true
    Module.Index0Float2Write=true
    Module.Index0Float3Write=true
    Module.Index1Float0Write=true
    Module.Index1Float1Write=true
    Module.Index1Float2Write=true
    Module.Index1Float3Write=true
    Module.Index2Float0Write=true
    Module.Index2Float1Write=true
    Module.Index2Float2Write=true
    Module.Index2Float3Write=true
    Write Parameter Index 1=::NewEnumerator1
    Write Parameter Index 2=::NewEnumerator1
   )
   &FN_Sprite_Attributes(
    Camera Offset:NiagaraFloat=&Multiply_Float(A=$Particles.RibbonWidth, B=1.5)
    Color=&Multiply_LinearColorByFloat(
     Linear Color=&VariantColorSelector(
      Variant Amount=4
      Original Color=<rgba 1 1 1 1>
      Variant 1 Color=<rgba 0.5304 0.5304 0.5304 1>
      Variant 2 Color=<rgba 0.2696 0.2696 0.2696 1>
      Variant 3 Color=<rgba 0.348 0.348 0.348 1>
      Variant 4 Color=<rgba 0.5304 0.5304 0.5304 1>
      Variant 5 Color=<rgba 1 1 1 1>
     )
     Float=$System.TODScalar
    )
    Create Camera Offset Attribute=true
    Create Color Attribute=true
    Module Output:ENiagara_AttributeSamplingApplyOutput=::NewEnumerator1
   )
   &DynamicMaterialParameters(
    Dynamic Parameter Output:ENiagara_AttributeSamplingApplyOutput=::NewEnumerator1
    Index 1 Param 2=2
    Index 1 Param 3=&FloatFromCurve(
     FloatCurve=DataInterface() {propertyValues {Curve {keys [
         {interpMode=RCIM_Cubic, leaveTangentWeight=0.1, tangentMode=RCTM_User, time=0, value=0.96}
         {arriveTangentWeight=0.2453, interpMode=RCIM_Cubic, leaveTangentWeight=0.3271, tangentMode=RCTM_User, time=0.3, value=-0.6231}
         {arriveTangentWeight=0.3297, interpMode=RCIM_Cubic, leaveTangentWeight=0.2473, tangentMode=RCTM_User, time=0.7, value=-0.608}
         {arriveTangentWeight=0.1, interpMode=RCIM_Cubic, tangentMode=RCTM_User, time=1, value=0.96}
        ]}}}
     CurveIndex=$Particles.ProgressAlpha
    )
    Module.Index0Float0Write=true
    Module.Index0Float1Write=true
    Module.Index0Float2Write=true
    Module.Index0Float3Write=true
    Module.Index1Float0Write=true
    Module.Index1Float1Write=true
    Module.Index1Float2Write=true
    Module.Index1Float3Write=true
    Module.Index2Float0Write=true
    Module.Index2Float1Write=true
    Module.Index2Float2Write=true
    Module.Index2Float3Write=true
    Write Parameter Index 1=::NewEnumerator1
    Write Parameter Index 2=::NewEnumerator1
   ) {module=DynamicMaterialParameters001}
  ]
  rendererClasses [&NiagaraRibbonRendererProperties]
  renderers [
   {rendererClass=&NiagaraRibbonRendererProperties}
   {rendererClass=&NiagaraRibbonRendererProperties, rendererIndex=1}
  ]
 }
 Glow_Packed {
  emitterUpdateScript [
   &EmitterState(
    Inactive Response=::NewEnumerator1
    Life Cycle Mode=::NewEnumerator1
    Loop Behavior=::NewEnumerator1
    Loop Duration Mode=::NewEnumerator1
    UseLoopDelay=false
   )
   &SpawnBurst_Instantaneous(
    Spawn Count=4
    Spawn Probability=&FN_PickFloatDependOnPlatform(TargetPlatform=DataInterface() {propertyValues {Platforms {qualityLevelMask=-2}}}, TargetPlatformActive=1, TargetPlatformNoActive=0)
    Use Spawn Probability=true
    UseLoopCountLimit=false
   ) {module=SpawnBurst_Instantaneous001}
   &SpawnBurst_Instantaneous(
    Spawn Group=1
    Spawn Probability=&FN_PickFloatDependOnPlatform(TargetPlatformActive=1, TargetPlatformNoActive=0)
    Use Spawn Probability=true
    UseLoopCountLimit=false
   )
   &SpawnBurst_Instantaneous(Spawn Group=2, Use Spawn Probability=false, UseLoopCountLimit=false) {module=SpawnBurst_Instantaneous002}
   &SpawnBurst_Instantaneous(Spawn Count=6, Spawn Group=3, Use Spawn Probability=false, UseLoopCountLimit=false) {module=SpawnBurst_Instantaneous003}
   &SpawnBurst_Instantaneous(Spawn Count=4, Spawn Group=4, Use Spawn Probability=false, UseLoopCountLimit=false) {module=SpawnBurst_Instantaneous004}
  ]
  particleSpawnScript [
   /SetVariables(Particles.VisibilityTag=$Emitter.SpawnGroup) {isSetParameters=true}
   &InitializeParticle(
    Color Mode=::NewEnumerator0
    Mesh Scale Mode=::NewEnumerator3
    Sprite Size Mode=::NewEnumerator3
    Uniform Sprite Size=7.7868
    UsePositionOffset=false
    Write Material Random=true
   ) {module=InitializeParticle001}
   &CameraOffset(Camera Offset Amount=&MergeHelper_FloatSwitch(0=1, 1=0, 2=0, 3=5, 4=10, Index=$Particles.VisibilityTag, Num=::NewEnumerator3))
   /SetVariables(Particles.LightSpecularScale=0.5) {isSetParameters=true}
  ]
  particleUpdateScript [
   &ParticleState(Kill Particles When Lifetime Has Elapsed=false)
   /SetVariables(Particles.GlowMeshPosition:NiagaraPosition=$Engine.Emitter.SimulationPosition) {isSetParameters=true}
   &SocketLocation(
    Orientation Sampling=::NewEnumerator1
    SamplePosOffset=true
    Sampled Position Offset=&MergeHelper_VectorSwitch(
     0=<0 0 0>
     1=<0 -7 5>
     2=<0 0 0>
     3=&MergeHelper_VectorSwitch(
      0=<0 0 -1.5>
      1=<0 0 -1.5>
      2=<0 0 -1.5>
      3=<0 0 -1.5>
      4=<0 0 0>
      5=<0 0 0.25>
      Index=/CascadeToNiagaraConverter/NiagaraScripts/DynamicInputScripts/SequentialInt(Use Array Data:NiagaraBool=false, First Number:NiagaraInt32=0, Sequence:NiagaraDataInterfaceArrayInt32=?, Last Number:NiagaraInt32=5)
      Num=::NewEnumerator4
     )
     4=<0 0 -0.3>
     Index=$Particles.VisibilityTag
     Num=::NewEnumerator3
    )
    Socket Index=&MergeHelper_IntSwitch(
     0=&SequentialInt(Use Array Data:NiagaraBool=false, First Number:NiagaraInt32=0, Sequence:NiagaraDataInterfaceArrayInt32=?, Last Number:NiagaraInt32=3)
     1=4
     2=0
     3=&SequentialInt(Use Array Data:NiagaraBool=false, First Number:NiagaraInt32=0, Sequence:NiagaraDataInterfaceArrayInt32=?, Last Number:NiagaraInt32=5)
     4=&SequentialInt(Use Array Data:NiagaraBool=false, First Number:NiagaraInt32=0, Sequence:NiagaraDataInterfaceArrayInt32=?, Last Number:NiagaraInt32=3)
     5=?
     Index=$Particles.VisibilityTag
     Num=::NewEnumerator3
    )
    Socket Reader=$System.Socket Reader
    Socket Sampling Mode=::NewEnumerator3
    VelocitySpeedLimit=false
   )
   &ScaleColor(
    Scale RGB=&MakeVectorFromLinearColorRGB(Color=&Multiply_LinearColorByFloat(
      Linear Color=&Multiply_LinearColorByFloat(
       Linear Color=&VariantColorSelector(
        Variant Amount=4
        Original Color=<rgba 1.1214 0.1821 3 1>
        Variant 1 Color=<rgba 3.1779 5.4893 10 1>
        Variant 2 Color=<rgba 1.5 0.75 0 1>
        Variant 3 Color=<rgba 11 0.6675 11 1>
        Variant 4 Color=<rgba 0 2 1.1667 1>
        Variant 5 Color=<rgba 1.0333 0 2 1>
       )
       Float=&FN_FrontEndFloatModifier(
        Gameplay=&FN_PickFloatDependOnPlatform(TargetPlatform=DataInterface() {propertyValues {Platforms {qualityLevelMask=-29}}}, TargetPlatformActive=1.2, TargetPlatformNoActive=1)
        Front End=&FN_PickFloatDependOnPlatform(TargetPlatform=DataInterface() {propertyValues {Platforms {qualityLevelMask=-29}}}, TargetPlatformActive=0.6, TargetPlatformNoActive=0.4)
       )
      )
      Float=&Multiply_Float(A=$System.TODScalar, B=$System.SystemFade)
     ))
    ScaleA=true
    ScaleRGB=true
   )
   &FN_Light_Attributes(
    Color=&VariantColorSelector(
     Variant Amount=4
     Original Color=<rgba 3.7379 0.6068 10 1>
     Variant 1 Color=<rgba 2.7012 4.6659 8.5 1>
     Variant 2 Color=<rgba 4 2.4974 0.2435 1>
     Variant 3 Color=<rgba 6.2506 2.3652 8.5 1>
     Variant 4 Color=<rgba 0 8.5 5.2417 1>
     Variant 5 Color=<rgba 4.4583 0.5 10 1>
    )
    Create Color Attribute=true
    Create Exponent Attribute=true
    Create Radius Attribute=true
    Exponent=5
    Radius=25
    Scale Color=&Multiply_Float(A=5, B=&Multiply_Float(A=$System.TODScalar, B=$System.SystemFade))
   )
   &FN_Mesh_Attributes(
    Color=&Multiply_LinearColorByFloat(
     Linear Color=&VariantColorSelector(
      Variant Amount=4
      Original Color=<rgba 2 0 2 1>
      Variant 1 Color=<rgba 0.5957 0.5957 1.5 1>
      Variant 2 Color=<rgba 2 1.3333 0 1>
      Variant 3 Color=<rgba 0.75 0.3 3 1>
      Variant 4 Color=<rgba 0.049 2 1.4435 1>
      Variant 5 Color=<rgba 1.11 0.9 3 1>
     )
     Float=&Multiply_Float(A=$System.TODScalar, B=$System.SystemFade)
    )
    Create Color Attribute=true
    Module Output=::NewEnumerator1
   )
   &FN_Mesh_Attributes(
    Color=&Multiply_LinearColorByFloat(
     Linear Color=&VariantColorSelector(
      Variant Amount=4
      Original Color=<rgba 0.7 0.0479 0.576 1>
      Variant 1 Color=<rgba 0.5043 0.6943 1 1>
      Variant 2 Color=<rgba 0.55 0.4059 0.0003 1>
      Variant 3 Color=<rgba 0.4456 0.5769 0.635 1>
      Variant 4 Color=<rgba 0.163 1.25 1.25 1>
      Variant 5 Color=<rgba 0.65 0.0649 0.5611 1>
     )
     Float=&Multiply_Float(A=$System.TODScalar, B=$System.SystemFade)
    )
    Create Color Attribute=true
    Module Output=::NewEnumerator1
   ) {module=FN_Mesh_Attributes001}
   &FN_Sprite_Attributes(
    Color=&MaskLinearColor(
     Mask Source=1
     Mask Parameter=&SequentialInt(Use Array Data:NiagaraBool=false, First Number:NiagaraInt32=0, Sequence:NiagaraDataInterfaceArrayInt32=?, Last Number:NiagaraInt32=5)
     Mask=4
     Masked Linear Color=&Multiply_LinearColorByFloat(
      Linear Color=&VariantColorSelector(
       Variant Amount=4
       Original Color=<rgba 1 0.1 1 0.05>
       Variant 1 Color=<rgba 0.4 0.61 1 0.05>
       Variant 2 Color=<rgba 1 0.64 0.1 0.05>
       Variant 3 Color=<rgba 1 0.1 1 0.1>
       Variant 4 Color=<rgba 0.0245 1 0.7218 0.05>
       Variant 5 Color=<rgba 0.5524 0.0739 1 0.085>
      )
      Float=&Multiply_Float(A=$System.TODScalar, B=$System.SystemFade)
     )
     Passthrough Linear Color=&Multiply_LinearColorByFloat(
      Linear Color=&VariantColorSelector(
       Variant Amount=4
       Original Color=<rgba 1 0.1 1 0.09>
       Variant 1 Color=<rgba 0.4 0.61 1 0.12>
       Variant 2 Color=<rgba 1 0.64 0.1 0.12>
       Variant 3 Color=<rgba 1 0.1 1 0.1>
       Variant 4 Color=<rgba 0.0245 1 0.7218 0.08>
       Variant 5 Color=<rgba 0.5524 0.0739 1 0.25>
      )
      Float=&Multiply_Float(A=$System.TODScalar, B=$System.SystemFade)
     )
    )
    Create Color Attribute=true
    Create SpriteSize Attribute=true
    Module Output:ENiagara_AttributeSamplingApplyOutput=::NewEnumerator1
    SpriteSize=&MergeHelper_Vector2Switch(
     0=<2 4>
     1=<2 4>
     2=<2 4>
     3=<2 4>
     4=&Vector2DFromFloat(Value=10)
     5=<2.5 4.5>
     Index=&SequentialInt(Use Array Data:NiagaraBool=false, First Number:NiagaraInt32=0, Sequence:NiagaraDataInterfaceArrayInt32=?, Last Number:NiagaraInt32=5)
     Num=::NewEnumerator4
    )
   )
   &DynamicMaterialParameters(
    Dynamic Parameter Output:ENiagara_AttributeSamplingApplyOutput=::NewEnumerator1
    Index 0 Param 1=0.5
    Index 0 Param 2=7
    Index 0 Param 3=1
    Module.Index0Float0Write=true
    Module.Index0Float1Write=true
    Module.Index0Float2Write=true
    Module.Index0Float3Write=true
   )
   &SpriteFacingAndAlignment(
    Sprite Alignment=&RotateVectorByQuaternion(Vector To Rotate=<0 0 1>, Quaternion=$Output.SocketLocation.SampledRotation)
    Sprite Facing=&RotateVectorByQuaternion(Vector To Rotate=<1 0 0>, Quaternion=$Output.SocketLocation.SampledRotation)
    WriteAlignment=true
    WriteFacing=true
   )
   &FN_Sprite_Attributes(
    Color=&Multiply_LinearColorByFloat(
     Linear Color=&VariantColorSelector(
      Variant Amount=4
      Original Color=<rgba 1.65 0.3 3 1>
      Variant 1 Color=<rgba 0.8125 1.4594 2.5 1>
      Variant 2 Color=<rgba 1 0.4333 0 1>
      Variant 3 Color=<rgba 2.5 0.6625 2.5 1>
      Variant 4 Color=<rgba 0.075 1 0.6146 1>
      Variant 5 Color=<rgba 2.317 0.26 4 1>
     )
     Float=&Multiply_Float(A=$System.TODScalar, B=$System.SystemFade)
    )
    Create Color Attribute=true
    Create Sprite Rotation Attribute=true
    Create SpriteSize Attribute=true
    Module Output:ENiagara_AttributeSamplingApplyOutput=::NewEnumerator1
    SpriteRotation:NiagaraFloat=&RandomRangeFloat(Minimum=0, Maximum=360)
    SpriteSize=&Vector2DFromFloat(Value=&Multiply_Float(
      A=3
      B=&Waveform(
       Waveform Count=::NewEnumerator1
       Global Input Position=&RandomRangeFloat(
        Minimum=0
        Maximum=5
        Randomness Mode=::NewEnumerator1
        Random Seed=&SequentialInt(Use Array Data:NiagaraBool=false, First Number:NiagaraInt32=0, Sequence:NiagaraDataInterfaceArrayInt32=?, Last Number:NiagaraInt32=3)
        Fixed Random Seed=true
       )
       "Waveform [1]"=::NewEnumerator0
       "[1] Frequency"=5
       "[1] Amplitude Min/Max"=<0.5 1>
       "[2] Average"=false
       "Waveform [2]"=::NewEnumerator5
       "[2] Input Position"=0
       "[2] Phase"=$Emitter.Age
       "[2] Phase Scale"=1
       "[2] Compound Frequency"=2
       "[2] Frequency"=2
       "[2] Amplitude Min/Max"=<0.4 1>
       "[2] Scale"=1
       "[2] Bias"=0
       "[2] Exponent"=1
       "[2] Pulse Duration"=0.1
       "[2] Clip Triangle Min/Max"=<0 0>
       "[2] Random Seed"=0
      )
     ))
   ) {module=FN_Sprite_Attributes001}
   &DynamicMaterialParameters(
    Dynamic Parameter Output:ENiagara_AttributeSamplingApplyOutput=::NewEnumerator1
    Index 0 Param 1=1
    Index 0 Param 2=1
    Index 0 Param 3=1
    Index 0 Param 4=1
    Module.Index0Float0Write=true
    Module.Index0Float1Write=true
    Module.Index0Float2Write=true
    Module.Index0Float3Write=true
   ) {module=DynamicMaterialParameters001}
  ]
  rendererClasses [
   &NiagaraLightRendererProperties
   &NiagaraSpriteRendererProperties
   &NiagaraMeshRendererProperties
  ]
  renderers [
   {rendererClass=&NiagaraLightRendererProperties, rendererIndex=0}
   {rendererClass=&NiagaraSpriteRendererProperties, rendererIndex=1}
   {rendererClass=&NiagaraMeshRendererProperties, rendererIndex=2}
   {rendererClass=&NiagaraSpriteRendererProperties, rendererIndex=3}
   {rendererClass=&NiagaraMeshRendererProperties, rendererIndex=4}
   {rendererClass=&NiagaraSpriteRendererProperties, rendererIndex=5}
  ]
 }
}
summary {
 systemName=NS_Backpack_VerTet_EnergyIdle
 emitters [
  {
   emitterName=Glow_Packed
   rendererClasses [
    &NiagaraLightRendererProperties
    &NiagaraSpriteRendererProperties
    &NiagaraMeshRendererProperties
   ]
  }
  {emitterName=Energy, rendererClasses [&NiagaraRibbonRendererProperties]}
  {emitterName=Center_Energy, rendererClasses [&NiagaraMeshRendererProperties]}
 ]
 userVariables [
  User.bIsFrontEnd {type=NiagaraBool}
  User.bIsFrontEndPreview {type=NiagaraBool}
  User.Variant {type=NiagaraFloat, defaultValue=0, description=""}
 ]
}
