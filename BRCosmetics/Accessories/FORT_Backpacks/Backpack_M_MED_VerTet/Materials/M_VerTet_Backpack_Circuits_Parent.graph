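# Additive, two-sided parent material for the animated glowing circuits on the VerTet backpack cosmetic, built between the 2023 character parent PreFX and PostFX functions.
# A procedural soft beam from mirrored UV0 is added to a light texture scrolling along UV1 V.
# Vertex red offsets the scroll time and varies its speed and direction.
# A UV1 edge fade masks the result, which drives opacity and, tinted by EmissiveColor, base color and emissive.
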Material /BRCosmetics/Accessories/FORT_Backpacks/Backpack_M_MED_VerTet/Materials/M_VerTet_Backpack_Circuits_Parent
graph {
 properties {
  bAutomaticallySetUsageInEditor=false
  bUseMaterialAttributes=true
  bUsedWithMeshDeformer=true
  bUsedWithMorphTargets=true
  bUsedWithSkeletalMesh=true
  blendMode=BLEND_Additive
  pixelDepthOffsetMode=PDOM_Legacy
  shadingModel=MSM_FromMaterialExpression
  twoSided=true
  usedShadingModels=Default Lit | Subsurface | Clear Coat | Subsurface Profile | Cloth
 }
 nodes [
  n1 = MaterialFunctionCall() {materialFunction=&MF_CharacterParent_2023_PreFX}
  n2 = TextureCoordinate() {uTiling=2}
  n3 = Abs(None=ConstantBiasScale(None=ComponentMask(None=@n2) {r=true}) {bias=-1, scale=2})
  n4 = ScalarParameter() {defaultValue=0.1, parameterName=BeamSize}
  n5 = Saturate(None=Divide(A=Subtract(A=@n3, B=@n4), B=OneMinus(None=@n4)))
  n6 = ScalarParameter() {defaultValue=1.3, parameterName=FalloffBeam}
  n7 = Divide(B=@n6)
  n8 = Power(Base=OneMinus(None=Power(Base=@n5, Exp=@n7)), Exp=@n6)
  n9 = Constant()
  n10 = VertexColor()
  n11 = Add(A=@n10.R) {constB=0.1}
  n12 = Multiply(
   A=Add(A=Time(), B=@n10.R)
   B=Multiply(A=LinearInterpolate(Alpha=@n11) {constA=-1}, B=ScalarParameter() {defaultValue=0.8, parameterName=CircuitDotsSpeed}) {constB=0.8}
  )
  n13 = TextureSample(UVs=Add(A=TextureCoordinate() {coordinateIndex=1}, B=AppendVector(A=@n9, B=@n12))) {
   samplerSource=SSM_Wrap_WorldGroupSettings
   texture=/BRCosmetics/Accessories/FORT_Backpacks/Backpack_M_MED_VerTet/Textures/T_Vertet_Backpack_CircuitLight
  }
  n14 = ComponentMask(None=TextureCoordinate() {coordinateIndex=1}) {g=true}
  n15 = Multiply(A=Multiply(A=@n14, B=OneMinus(None=@n14))) {constB=3}
  n16 = Multiply(A=Add(A=Saturate(None=@n8), B=@n13.R), B=Saturate(None=Power(Base=@n15) {constExponent=0.5}))
  n17 = Multiply(A=@n16, B=VectorParameter() {parameterName=EmissiveColor, defaultValue=<rgba 1 0 0.8079 1>}.RGB)
  MaterialFunctionCall(
   Material=SetMaterialAttributes!(MaterialAttributes=@n1.Result, Base Color=@n17, Emissive Color=@n17, Opacity=@n16) {attributeSetTypes [#1, #2, #3]}
   M=@n1.M
   Pre Skinned Local Position=@n1.M
   Distance Blend=@n1.M
   Fresnel=@n1.M
  ) {materialFunction=&MF_CharacterParent_2023_PostFX}
 ]
}
